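# Player animation blueprint for the OctopusVehicle driver, supplying the VehicleFinalPoseOverride layer pose; the other override layers are empty.
# A locomotion state machine (directional blendspace, pivots, boosting, grapple fire) feeds a steer-and-pitch head aim offset.
# An additive state machine layers idle, leans and velocity-impact collision reactions, followed by a gender-specific pose additive.
# Root and IK bones then snap to world-space seat and attachment transforms, with Leg IK nodes solving both hands and feet.
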AnimBlueprint /Game/Animation/Game/MainPlayer/Vehicles/OctopusVehicle/OctopusVehicle_Driver_AnimBP
target_skeleton=&Fortnite_M_Avg_Player_Skeleton
graphs [
 AnimGraph {nodes [Output Pose() {layerGroup=None}]}
 EventGraph {links [], nodes []}
 VehicleFinalPoseOverride {nodes [
   Input Pose() {name=InFinalPose}
   AnimGraphNode_StateMachine_2 = Locomotion() {is=title AnimGraphNode_StateMachine}
   n1 = AimOffset Player 'OctopusVehicle_Head_AO'!(
    X:float=Get SeatSteerYaw() -> (SeatSteerYaw:float)
    Y:float=Get AimPitch()
    BasePose:Pose Link Structure=@AnimGraphNode_StateMachine_2
    Alpha:float=1
   ) -> (Pose:Pose Link Structure) {blendSpace=/Game/Animation/Game/MainPlayer/Vehicles/OctopusVehicle/AO/OctopusVehicle_Head_AO, alphaBoolBlend {blendOption=Linear}}
   AnimGraphNode_StateMachine_0 = FullBodyAdditive() {is=title AnimGraphNode_StateMachine}
   n2 = Get bHasDriver() -> (bHasDriver:bool)
   n3 = Apply Additive!(Base=@n1, Additive=@AnimGraphNode_StateMachine_0, bAlphaBoolEnabled:bool=@n2) -> (Pose) {alphaInputType=Bool, alphaBoolBlend {blendOption=Linear, blendOutTime=0.15}}
   n4 = "Blend Poses (EFortCustomGender)"!(
    BlendPose_0:Pose Link Structure="Sequence Player (additive) 'OctopusVehicle_Idle_Pose_Additive'"() {sequence=&/OctopusVehicle_Idle_Pose_Additive}
    BlendPose_1:Pose Link Structure="Sequence Player (additive) 'OctopusVehicle_Idle_Pose_CMF'"() {sequence=&/OctopusVehicle_Idle_Pose_CMF}
    BlendTime_0:float=0.1
    BlendTime_1:float=0.1
    ActiveEnumValue:Byte=Get Gender() -> (Gender:EFortCustomGender Enum=Invalid)
   ) -> (Pose:Pose Link Structure) {
    is=title AnimGraphNode_BlendListByEnum
    blendType=Linear
    blendPose [{linkId=-1, sourceLinkId=-1}, {linkId=-1, sourceLinkId=-1}]
    blendTime [0.1,0.1]
   }
   n5 = Get bHasDriver()
   n6 = "Transform (Modify) Bone - Bone: Root"!(
    ComponentPose=Local To Component(LocalPose=Apply Additive(Base=@n3, Additive=@n4, bAlphaBoolEnabled=@n2) {alphaInputType=Bool, alphaBoolBlend {blendOption=Linear, blendOutTime=0.15}})
    bAlphaBoolEnabled:bool=@n5
    Translation=Get SeatLocation() -> (SeatLocation:Vector)
    Rotation=Get SeatRotationPose() -> (SeatRotationPose:Rotator)
   ) -> (Pose) {
    alphaInputType=Bool
    rotationMode=BMM_Replace
    rotationSpace=BCS_WorldSpace
    translationMode=BMM_Replace
    translationSpace=BCS_WorldSpace
    alphaBoolBlend {blendOption=Linear}
    boneToModify {boneName=Root}
   }
   n7 = "Transform (Modify) Bone - Bone: ik_hand_l"!(
    ComponentPose=@n6
    bAlphaBoolEnabled:bool=@n5
    Translation=Get HandAttachmentLocLeft() -> (HandAttachmentLocLeft:Vector)
    Rotation=Get HandAttachmentRotLeft() -> (HandAttachmentRotLeft:Rotator)
   ) -> (Pose) {
    alphaInputType=Bool
    rotationMode=BMM_Replace
    rotationSpace=BCS_WorldSpace
    translationMode=BMM_Replace
    translationSpace=BCS_WorldSpace
    alphaBoolBlend {blendOption=Linear}
    boneToModify {boneName=ik_hand_l}
   }
   n8 = "Transform (Modify) Bone - Bone: ik_hand_r"!(
    ComponentPose=@n7
    bAlphaBoolEnabled:bool=@n5
    Translation=Get HandAttachmentLocRight() -> (HandAttachmentLocRight:Vector)
    Rotation=Get HandAttachmentRotRight() -> (HandAttachmentRotRight:Rotator)
   ) -> (Pose) {
    alphaInputType=Bool
    rotationMode=BMM_Replace
    rotationSpace=BCS_WorldSpace
    translationMode=BMM_Replace
    translationSpace=BCS_WorldSpace
    alphaBoolBlend {blendOption=Linear}
    boneToModify {boneName=ik_hand_r}
   }
   n9 = "Transform (Modify) Bone - Bone: ik_foot_l"!(
    ComponentPose=@n8
    bAlphaBoolEnabled:bool=@n5
    Translation=Get FootAttachmentLocLeft() -> (FootAttachmentLocLeft:Vector)
    Rotation=Get FootAttachmentRotLeft() -> (FootAttachmentRotLeft:Rotator)
   ) -> (Pose) {
    alphaInputType=Bool
    rotationMode=BMM_Replace
    rotationSpace=BCS_WorldSpace
    translationMode=BMM_Replace
    translationSpace=BCS_WorldSpace
    alphaBoolBlend {blendOption=Linear}
    boneToModify {boneName=ik_foot_l}
   }
   n10 = "Transform (Modify) Bone - Bone: ik_foot_r"!(
    ComponentPose=@n9
    bAlphaBoolEnabled:bool=@n5
    Translation=Get FootAttachmentLocRight() -> (FootAttachmentLocRight:Vector)
    Rotation=Get FootAttachmentRotRight() -> (FootAttachmentRotRight:Rotator)
   ) -> (Pose) {
    alphaInputType=Bool
    rotationMode=BMM_Replace
    rotationSpace=BCS_WorldSpace
    translationMode=BMM_Replace
    translationSpace=BCS_WorldSpace
    alphaBoolBlend {blendOption=Linear}
    boneToModify {boneName=ik_foot_r}
   }
   n11 = Leg IK!(ComponentPose:Component Space Pose Link Structure=@n10, Alpha:float=Get RightHandIKAlpha() -> (RightHandIKAlpha:float)) -> (Pose:Component Space Pose Link Structure) {
    is=title AnimGraphNode_LegIK
    alphaBoolBlend {blendOption=Linear}
    legsDefinition [{
      bEnableKneeTwistCorrection=false
      bEnableRotationLimit=false
      footBoneForwardAxis=Y
      hingeRotationAxis=Z
      minRotationAngle=15
      numBonesInLimb=2
      twistOffsetCurveName=None
      fKFootBone {boneName=hand_r}
      iKFootBone {boneName=ik_hand_r}
     }]
   }
   n12 = Leg IK(ComponentPose=@n11, Alpha=Get LeftHandIKAlpha() -> (LeftHandIKAlpha:float)) {
    alphaBoolBlend {blendOption=Linear}
    legsDefinition [{
      bEnableKneeTwistCorrection=false
      bEnableRotationLimit=false
      footBoneForwardAxis=Y
      hingeRotationAxis=Z
      minRotationAngle=15
      numBonesInLimb=2
      twistOffsetCurveName=None
      fKFootBone {boneName=hand_l}
      iKFootBone {boneName=ik_hand_l}
     }]
   }
   n13 = Leg IK(ComponentPose=@n12, Alpha=Get LegIKAlpha() -> (LegIKAlpha:float)) {
    alphaBoolBlend {blendOption=Linear}
    legsDefinition [
     {
      bEnableKneeTwistCorrection=false
      bEnableRotationLimit=false
      footBoneForwardAxis=Y
      hingeRotationAxis=Z
      minRotationAngle=15
      numBonesInLimb=2
      twistOffsetCurveName=None
      fKFootBone {boneName=foot_l}
      iKFootBone {boneName=ik_foot_l}
     }
     {
      bEnableKneeTwistCorrection=false
      bEnableRotationLimit=false
      footBoneForwardAxis=Y
      hingeRotationAxis=Z
      minRotationAngle=15
      numBonesInLimb=2
      twistOffsetCurveName=None
      fKFootBone {boneName=foot_r}
      iKFootBone {boneName=ik_foot_r}
     }
    ]
   }
   Output Pose(Result=Component To Local(ComponentPose=@n13)) {layerGroup=VehicleOverride}
   Save cached pose 'DirectionPose'(Pose=Blendspace Player 'OctopusVehicle_Direction_BS'(X=Get LeftRight() -> (LeftRight:float), Y=Get FwdBwd() -> (FwdBwd:float)) {blendSpace=&/OctopusVehicle_Direction_BS}) {cacheName=DirectionPose}
  ]}
 FullBodyAdditive {
  entry=Idle
  path=VehicleFinalPoseOverride.AnimGraphNode_StateMachine_0.FullBodyAdditive
  ids=id
  nodes [
   AnimStateNode_0 = Idle()
   AnimStateNode_1 = Collision_Bottom() {is=title AnimStateNode}
   AnimStateNode_2 = Collision_NS() {is=title AnimStateNode}
   AnimStateNode_3 = Collision_E() {is=title AnimStateNode}
   AnimStateNode_4 = Collision_W() {is=title AnimStateNode}
   AnimStateTransitionNode_0 = AnimStateTransitionNode() {bAutomaticRuleBasedOnSequencePlayerInState=true}
   AnimStateTransitionNode_1 = AnimStateTransitionNode() {bAutomaticRuleBasedOnSequencePlayerInState=true}
   AnimStateTransitionNode_2 = AnimStateTransitionNode() {bAutomaticRuleBasedOnSequencePlayerInState=true}
   AnimStateTransitionNode_3 = AnimStateTransitionNode() {bAutomaticRuleBasedOnSequencePlayerInState=true}
   AnimStateTransitionNode_4 = AnimStateTransitionNode() {crossfadeDuration=0}
   AnimStateTransitionNode_5 = AnimStateTransitionNode() {crossfadeDuration=0}
   AnimStateTransitionNode_6 = AnimStateTransitionNode() {crossfadeDuration=0}
   AnimStateTransitionNode_7 = AnimStateTransitionNode() {crossfadeDuration=0}
   AnimStateTransitionNode_8 = AnimStateTransitionNode() {crossfadeDuration=0}
  ]
  links [
   {
    dst=Idle
    rule=VehicleFinalPoseOverride.AnimGraphNode_StateMachine_0.FullBodyAdditive.AnimStateTransitionNode_0.Transition
    src=Collision_Bottom
    via=AnimStateTransitionNode_0
   }
   {
    dst=Idle
    rule=VehicleFinalPoseOverride.AnimGraphNode_StateMachine_0.FullBodyAdditive.AnimStateTransitionNode_1.Transition
    src=Collision_E
    via=AnimStateTransitionNode_1
   }
   {
    dst=Idle
    rule=VehicleFinalPoseOverride.AnimGraphNode_StateMachine_0.FullBodyAdditive.AnimStateTransitionNode_2.Transition
    src=Collision_NS
    via=AnimStateTransitionNode_2
   }
   {
    dst=Idle
    rule=VehicleFinalPoseOverride.AnimGraphNode_StateMachine_0.FullBodyAdditive.AnimStateTransitionNode_3.Transition
    src=Collision_W
    via=AnimStateTransitionNode_3
   }
   {
    dst=Collision_Bottom
    rule=VehicleFinalPoseOverride.AnimGraphNode_StateMachine_0.FullBodyAdditive.AnimStateTransitionNode_4.Transition
    src=Idle
    via=AnimStateTransitionNode_4
   }
   {
    dst=Collision_E
    rule=VehicleFinalPoseOverride.AnimGraphNode_StateMachine_0.FullBodyAdditive.AnimStateTransitionNode_5.Transition
    src=Idle
    via=AnimStateTransitionNode_5
   }
   {
    dst=Collision_NS
    rule=VehicleFinalPoseOverride.AnimGraphNode_StateMachine_0.FullBodyAdditive.AnimStateTransitionNode_6.Transition
    src=Idle
    via=AnimStateTransitionNode_6
   }
   {
    dst=Collision_NS
    rule=VehicleFinalPoseOverride.AnimGraphNode_StateMachine_0.FullBodyAdditive.AnimStateTransitionNode_8.Transition
    src=Idle
    via=AnimStateTransitionNode_8
   }
   {
    dst=Collision_W
    rule=VehicleFinalPoseOverride.AnimGraphNode_StateMachine_0.FullBodyAdditive.AnimStateTransitionNode_7.Transition
    src=Idle
    via=AnimStateTransitionNode_7
   }
  ]
 }
 Idle {
  path=VehicleFinalPoseOverride.AnimGraphNode_StateMachine_0.FullBodyAdditive.AnimStateNode_0.Idle
  nodes [
   n1 = "Sequence Player (additive) 'OctopusVehicle_Idle'"() {groupName=Idle, method=SyncGroup, sequence=&/OctopusVehicle_Idle}
   n2 = "Sequence Player (additive) 'OctopusVehicle_Idle_Fast'"() {groupName=Idle, method=SyncGroup, sequence=&/OctopusVehicle_Idle_Fast}
   n3 = Get bIsLowerVelocity() -> (bIsLowerVelocity:bool)
   n4 = Get SeatSteerStrength() -> (SeatSteerStrength:float)
   Output Animation Pose(Result=Apply Additive(
     Base=Blend Poses by bool(BlendPose_0=@n1, BlendPose_1=@n2, BlendTime_0=0.4, BlendTime_1=0.15, bActiveValue=@n3) {blendType=Linear}
     Additive=Blendspace Player 'OctopusVehicle_Leans_BS'!(X=@n4) -> (Pose) {blendSpace=&/OctopusVehicle_Leans_BS}
     bAlphaBoolEnabled=Get bShouldApplyLeanAdditive() -> (bShouldApplyLeanAdditive:bool)
    ) {alphaInputType=Bool, alphaBoolBlend {blendInTime=0.15, blendOption=Linear, blendOutTime=0.15}})
  ]
 }
 Collision_Bottom {
  path=VehicleFinalPoseOverride.AnimGraphNode_StateMachine_0.FullBodyAdditive.AnimStateNode_1.Collision_Bottom
  nodes [Output Animation Pose(Result="Sequence Player (additive) 'OctopusVehicle_Collision_N'"() {bLoopAnimation=false, sequence=&/OctopusVehicle_Collision_N})]
 }
 Collision_NS {
  path=VehicleFinalPoseOverride.AnimGraphNode_StateMachine_0.FullBodyAdditive.AnimStateNode_2.Collision_NS
  nodes [Output Animation Pose(Result="Sequence Player (additive) 'OctopusVehicle_Collision_N'"() {bLoopAnimation=false, sequence=&OctopusVehicle_Collision_N})]
 }
 Collision_E {
  path=VehicleFinalPoseOverride.AnimGraphNode_StateMachine_0.FullBodyAdditive.AnimStateNode_3.Collision_E
  nodes [Output Animation Pose(Result="Sequence Player (additive) 'OctopusVehicle_Collision_E'"() {bLoopAnimation=false, sequence=&/OctopusVehicle_Collision_E})]
 }
 Collision_W {
  path=VehicleFinalPoseOverride.AnimGraphNode_StateMachine_0.FullBodyAdditive.AnimStateNode_4.Collision_W
  nodes [Output Animation Pose(Result="Sequence Player (additive) 'OctopusVehicle_Collision_W'"() {bLoopAnimation=false, sequence=&/OctopusVehicle_Collision_W})]
 }
 Transition {path=VehicleFinalPoseOverride.AnimGraphNode_StateMachine_0.FullBodyAdditive.AnimStateTransitionNode_0.Transition, nodes [Result()]}
 Transition {path=VehicleFinalPoseOverride.AnimGraphNode_StateMachine_0.FullBodyAdditive.AnimStateTransitionNode_1.Transition, nodes [Result()]}
 Transition {path=VehicleFinalPoseOverride.AnimGraphNode_StateMachine_0.FullBodyAdditive.AnimStateTransitionNode_2.Transition, nodes [Result()]}
 Transition {path=VehicleFinalPoseOverride.AnimGraphNode_StateMachine_0.FullBodyAdditive.AnimStateTransitionNode_3.Transition, nodes [Result()]}
 Transition {
  path=VehicleFinalPoseOverride.AnimGraphNode_StateMachine_0.FullBodyAdditive.AnimStateTransitionNode_4.Transition
  nodes [Result(bCanEnterTransition=Break Fort Anim Input Velocity Impact!(FortAnimInput_VelocityImpact:"Fort Anim Input Velocity Impact Structure (by ref)"=Get VelocityImpact() -> (VelocityImpact:Fort Anim Input Velocity Impact Structure)) -> (
     DeltaVelocityThreshold:Vector=~
     ImpactScale:Vector=~
     ImpactLimitX:Input Range Structure
     ImpactLimitY:Input Range Structure
     ImpactLimitZ:Input Range Structure
     SpringInterpolatorX:Float RK4Spring Interpolator Structure
     SpringInterpolatorY:Float RK4Spring Interpolator Structure
     SpringInterpolatorZ:Float RK4Spring Interpolator Structure
     bTestVelocity:bool=~
     TestVelocity:Vector=~
     bIsForwardImpact:bool=~
     bIsBackwardImpact:bool=~
     bIsLeftImpact:bool=~
     bIsRightImpact:bool=~
     bIsUpImpact:bool=~
     bIsDownImpact:bool=~
    ) {is=title K2Node_BreakStruct}.bIsDownImpact)]
 }
 Transition {
  path=VehicleFinalPoseOverride.AnimGraphNode_StateMachine_0.FullBodyAdditive.AnimStateTransitionNode_5.Transition
  nodes [Result(bCanEnterTransition=Break Fort Anim Input Velocity Impact(FortAnimInput_VelocityImpact=Get VelocityImpact()) -> (
     DeltaVelocityThreshold=~
     ImpactScale=~
     bTestVelocity=~
     TestVelocity=~
     bIsForwardImpact=~
     bIsBackwardImpact=~
     bIsLeftImpact=~
     bIsRightImpact=~
     bIsUpImpact=~
     bIsDownImpact=~
    ).bIsRightImpact)]
 }
 Transition {
  path=VehicleFinalPoseOverride.AnimGraphNode_StateMachine_0.FullBodyAdditive.AnimStateTransitionNode_6.Transition
  nodes [Result(bCanEnterTransition=Break Fort Anim Input Velocity Impact(FortAnimInput_VelocityImpact=Get VelocityImpact()) -> (
     DeltaVelocityThreshold=~
     ImpactScale=~
     bTestVelocity=~
     TestVelocity=~
     bIsForwardImpact=~
     bIsBackwardImpact=~
     bIsLeftImpact=~
     bIsRightImpact=~
     bIsUpImpact=~
     bIsDownImpact=~
    ).bIsForwardImpact)]
 }
 Transition {
  path=VehicleFinalPoseOverride.AnimGraphNode_StateMachine_0.FullBodyAdditive.AnimStateTransitionNode_7.Transition
  nodes [Result(bCanEnterTransition=Break Fort Anim Input Velocity Impact(FortAnimInput_VelocityImpact=Get VelocityImpact()) -> (
     DeltaVelocityThreshold=~
     ImpactScale=~
     bTestVelocity=~
     TestVelocity=~
     bIsForwardImpact=~
     bIsBackwardImpact=~
     bIsLeftImpact=~
     bIsRightImpact=~
     bIsUpImpact=~
     bIsDownImpact=~
    ).bIsLeftImpact)]
 }
 Transition {
  path=VehicleFinalPoseOverride.AnimGraphNode_StateMachine_0.FullBodyAdditive.AnimStateTransitionNode_8.Transition
  nodes [Result(bCanEnterTransition=Break Fort Anim Input Velocity Impact(FortAnimInput_VelocityImpact=Get VelocityImpact()) -> (
     DeltaVelocityThreshold=~
     ImpactScale=~
     bTestVelocity=~
     TestVelocity=~
     bIsForwardImpact=~
     bIsBackwardImpact=~
     bIsLeftImpact=~
     bIsRightImpact=~
     bIsUpImpact=~
     bIsDownImpact=~
    ).bIsBackwardImpact)]
 }
 Locomotion {
  entry=Default_InPlace
  path=VehicleFinalPoseOverride.AnimGraphNode_StateMachine_2.Locomotion
  ids=id
  nodes [
   AnimStateNode_0 = GrapplerFireFromBoosting() {is=title AnimStateNode}
   AnimStateNode_1 = Boosting_Loop() {is=title AnimStateNode}
   AnimStateNode_2 = Default_InPlace() {is=title AnimStateNode}
   AnimStateNode_3 = Pivot() {is=title AnimStateNode}
   AnimStateNode_4 = Default() {is=title AnimStateNode}
   AnimStateNode_5 = Boosting_Intro() {is=title AnimStateNode}
   AnimStateNode_6 = GrappleFireIntro() {is=title AnimStateNode}
   AnimStateNode_7 = GrappleFireLoop() {is=title AnimStateNode}
   AnimStateTransitionNode_0 = AnimStateTransitionNode() {bAutomaticRuleBasedOnSequencePlayerInState=true, crossfadeDuration=0, priorityOrder=2}
   AnimStateTransitionNode_1 = AnimStateTransitionNode() {crossfadeDuration=0}
   AnimStateTransitionNode_10 = AnimStateTransitionNode() {bAutomaticRuleBasedOnSequencePlayerInState=true, crossfadeDuration=0.15}
   AnimStateTransitionNode_11 = AnimStateTransitionNode()
   AnimStateTransitionNode_12 = AnimStateTransitionNode() {crossfadeDuration=0.1}
   AnimStateTransitionNode_13 = AnimStateTransitionNode() {crossfadeDuration=0.15}
   AnimStateTransitionNode_14 = AnimStateTransitionNode() {crossfadeDuration=0.067}
   AnimStateTransitionNode_15 = AnimStateTransitionNode() {crossfadeDuration=0.067}
   AnimStateTransitionNode_16 = AnimStateTransitionNode() {crossfadeDuration=0.15}
   AnimStateTransitionNode_17 = AnimStateTransitionNode() {crossfadeDuration=0.15}
   AnimStateTransitionNode_18 = AnimStateTransitionNode() {crossfadeDuration=0.15}
   AnimStateTransitionNode_19 = AnimStateTransitionNode() {crossfadeDuration=0.15}
   AnimStateTransitionNode_2 = AnimStateTransitionNode() {crossfadeDuration=0.45, priorityOrder=2}
   AnimStateTransitionNode_20 = AnimStateTransitionNode() {bAutomaticRuleBasedOnSequencePlayerInState=true}
   AnimStateTransitionNode_4 = AnimStateTransitionNode() {bAutomaticRuleBasedOnSequencePlayerInState=true, crossfadeDuration=0.45}
   AnimStateTransitionNode_5 = AnimStateTransitionNode() {crossfadeDuration=0}
   AnimStateTransitionNode_6 = AnimStateTransitionNode()
   AnimStateTransitionNode_7 = AnimStateTransitionNode()
   AnimStateTransitionNode_8 = AnimStateTransitionNode()
   AnimStateTransitionNode_9 = AnimStateTransitionNode() {crossfadeDuration=0.15}
  ]
  links [
   {
    dst=Boosting_Loop
    rule=VehicleFinalPoseOverride.AnimGraphNode_StateMachine_2.Locomotion.AnimStateTransitionNode_0.AnimationTransitionGraph_0
    src=Boosting_Intro
    via=AnimStateTransitionNode_0
   }
   {
    dst=GrapplerFireFromBoosting
    rule=VehicleFinalPoseOverride.AnimGraphNode_StateMachine_2.Locomotion.AnimStateTransitionNode_18.Transition
    src=Boosting_Intro
    via=AnimStateTransitionNode_18
   }
   {
    dst=Default
    rule=VehicleFinalPoseOverride.AnimGraphNode_StateMachine_2.Locomotion.AnimStateTransitionNode_2.Transition
    src=Boosting_Loop
    via=AnimStateTransitionNode_2
   }
   {
    dst=GrapplerFireFromBoosting
    rule=VehicleFinalPoseOverride.AnimGraphNode_StateMachine_2.Locomotion.AnimStateTransitionNode_19.Transition
    src=Boosting_Loop
    via=AnimStateTransitionNode_19
   }
   {
    dst=Boosting_Intro
    rule=VehicleFinalPoseOverride.AnimGraphNode_StateMachine_2.Locomotion.AnimStateTransitionNode_1.AnimationTransitionGraph_1
    src=Default
    via=AnimStateTransitionNode_1
   }
   {
    dst=Default_InPlace
    rule=VehicleFinalPoseOverride.AnimGraphNode_StateMachine_2.Locomotion.AnimStateTransitionNode_7.Transition
    src=Default
    via=AnimStateTransitionNode_7
   }
   {
    dst=GrappleFireIntro
    rule=VehicleFinalPoseOverride.AnimGraphNode_StateMachine_2.Locomotion.AnimStateTransitionNode_9.Transition
    src=Default
    via=AnimStateTransitionNode_9
   }
   {
    dst=GrappleFireLoop
    rule=VehicleFinalPoseOverride.AnimGraphNode_StateMachine_2.Locomotion.AnimStateTransitionNode_16.Transition
    src=Default
    via=AnimStateTransitionNode_16
   }
   {
    dst=Pivot
    rule=VehicleFinalPoseOverride.AnimGraphNode_StateMachine_2.Locomotion.AnimStateTransitionNode_8.Transition
    src=Default
    via=AnimStateTransitionNode_8
   }
   {
    dst=Default
    rule=VehicleFinalPoseOverride.AnimGraphNode_StateMachine_2.Locomotion.AnimStateTransitionNode_6.Transition
    src=Default_InPlace
    via=AnimStateTransitionNode_6
   }
   {
    dst=GrappleFireIntro
    rule=VehicleFinalPoseOverride.AnimGraphNode_StateMachine_2.Locomotion.AnimStateTransitionNode_13.Transition
    src=Default_InPlace
    via=AnimStateTransitionNode_13
   }
   {
    dst=GrappleFireLoop
    rule=VehicleFinalPoseOverride.AnimGraphNode_StateMachine_2.Locomotion.AnimStateTransitionNode_17.Transition
    src=Default_InPlace
    via=AnimStateTransitionNode_17
   }
   {
    dst=Boosting_Intro
    rule=VehicleFinalPoseOverride.AnimGraphNode_StateMachine_2.Locomotion.AnimStateTransitionNode_14.Transition
    src=GrappleFireIntro
    via=AnimStateTransitionNode_14
   }
   {
    dst=GrappleFireLoop
    rule=VehicleFinalPoseOverride.AnimGraphNode_StateMachine_2.Locomotion.AnimStateTransitionNode_10.Transition
    src=GrappleFireIntro
    via=AnimStateTransitionNode_10
   }
   {
    dst=Boosting_Intro
    rule=VehicleFinalPoseOverride.AnimGraphNode_StateMachine_2.Locomotion.AnimStateTransitionNode_15.Transition
    src=GrappleFireLoop
    via=AnimStateTransitionNode_15
   }
   {
    dst=Default
    rule=VehicleFinalPoseOverride.AnimGraphNode_StateMachine_2.Locomotion.AnimStateTransitionNode_11.Transition
    src=GrappleFireLoop
    via=AnimStateTransitionNode_11
   }
   {
    dst=Boosting_Loop
    rule=VehicleFinalPoseOverride.AnimGraphNode_StateMachine_2.Locomotion.AnimStateTransitionNode_20.Transition
    src=GrapplerFireFromBoosting
    via=AnimStateTransitionNode_20
   }
   {
    dst=Boosting_Intro
    rule=VehicleFinalPoseOverride.AnimGraphNode_StateMachine_2.Locomotion.AnimStateTransitionNode_5.Transition
    src=Pivot
    via=AnimStateTransitionNode_5
   }
   {
    dst=Default
    rule=VehicleFinalPoseOverride.AnimGraphNode_StateMachine_2.Locomotion.AnimStateTransitionNode_4.Transition
    src=Pivot
    via=AnimStateTransitionNode_4
   }
   {
    dst=GrappleFireIntro
    rule=VehicleFinalPoseOverride.AnimGraphNode_StateMachine_2.Locomotion.AnimStateTransitionNode_12.Transition
    src=Pivot
    via=AnimStateTransitionNode_12
   }
  ]
 }
 GrapplerFireFromBoosting {
  path=VehicleFinalPoseOverride.AnimGraphNode_StateMachine_2.Locomotion.AnimStateNode_0.GrapplerFireFromBoosting
  nodes [
   n1 = Sequence Player 'OctopusVehicle_GrappleFire_Boosting'() {bLoopAnimation=false, playRate=1.1, sequence=&/OctopusVehicle_GrappleFire_Boosting}
   Output Animation Pose(Result=@n1)
  ]
 }
 Boosting_Loop {
  path=VehicleFinalPoseOverride.AnimGraphNode_StateMachine_2.Locomotion.AnimStateNode_1.Boosting_Loop
  nodes [Output Animation Pose(Result=Sequence Player 'OctopusVehicle_Boosting_Pose'() {sequence=&/OctopusVehicle_Boosting_Pose})]
 }
 Default_InPlace {
  path=VehicleFinalPoseOverride.AnimGraphNode_StateMachine_2.Locomotion.AnimStateNode_2.Default_InPlace
  nodes [Output Animation Pose(Result=Use cached pose 'DirectionPose'())]
 }
 Pivot {
  path=VehicleFinalPoseOverride.AnimGraphNode_StateMachine_2.Locomotion.AnimStateNode_3.Pivot
  nodes [
   n1 = "Blend Poses (EFortCustomGender)"(
    BlendPose_0=Sequence Player 'OctopusVehicle_Pivot_StoN'() {bLoopAnimation=false, sequence=&/OctopusVehicle_Pivot_StoN}
    BlendPose_1=Sequence Player 'OctopusVehicle_Pivot_StoN_CMF'() {bLoopAnimation=false, sequence=&/OctopusVehicle_Pivot_StoN_CMF}
    BlendTime_0=0.1
    BlendTime_1=0.1
    ActiveEnumValue=Get Gender() -> (Gender=Invalid)
   ) {blendType=Linear, blendPose [{linkId=-1, sourceLinkId=-1}, {linkId=-1, sourceLinkId=-1}], blendTime [0.1,0.1]}
   n2 = "Blend Poses (EFortCardinalDirection)"!(
    BlendPose_0:Pose Link Structure=Sequence Player 'OctopusVehicle_Pivot_NtoS'() {bLoopAnimation=false, sequence=&/OctopusVehicle_Pivot_NtoS}
    BlendPose_1:Pose Link Structure=@n1
    BlendPose_2:Pose Link Structure=Sequence Player 'OctopusVehicle_Pivot_WtoE'() {bLoopAnimation=false, sequence=&/OctopusVehicle_Pivot_WtoE}
    BlendPose_3:Pose Link Structure=Sequence Player 'OctopusVehicle_Pivot_EtoW'() {bLoopAnimation=false, sequence=&/OctopusVehicle_Pivot_EtoW}
    BlendTime_0:float=0.3
    BlendTime_1:float=0.3
    BlendTime_2:float=0.3
    BlendTime_3:float=0.3
    ActiveEnumValue:Byte=Get PivotDir() -> (PivotDir:EFortCardinalDirection Enum=North)
   ) -> (Pose:Pose Link Structure) {
    is=title AnimGraphNode_BlendListByEnum
    blendType=Linear
    blendPose [{linkId=-1, sourceLinkId=-1}, {linkId=-1, sourceLinkId=-1}, {linkId=-1, sourceLinkId=-1}, {linkId=-1, sourceLinkId=-1}]
    blendTime [0.1,0.1,0.1,0.1]
   }
   Output Animation Pose(Result=@n2)
  ]
 }
 Default {
  path=VehicleFinalPoseOverride.AnimGraphNode_StateMachine_2.Locomotion.AnimStateNode_4.Default
  nodes [Output Animation Pose(Result=Use cached pose 'DirectionPose'())]
 }
 Boosting_Intro {
  path=VehicleFinalPoseOverride.AnimGraphNode_StateMachine_2.Locomotion.AnimStateNode_5.Boosting_Intro
  nodes [
   n1 = Sequence Player 'OctopusVehicle_Boosting_Intro'!(PlayRate:float=1.2) -> (Pose) {bLoopAnimation=false, playRate=1.2, sequence=&/OctopusVehicle_Boosting_Intro}
   Output Animation Pose(Result=@n1)
  ]
 }
 GrappleFireIntro {
  path=VehicleFinalPoseOverride.AnimGraphNode_StateMachine_2.Locomotion.AnimStateNode_6.GrappleFireIntro
  nodes [
   n1 = Sequence Player 'OctopusVehicle_GrappleFire_Intro'() {bLoopAnimation=false, sequence=&/OctopusVehicle_GrappleFire_Intro, startPosition=0.2}
   Output Animation Pose(Result=@n1)
  ]
 }
 GrappleFireLoop {
  path=VehicleFinalPoseOverride.AnimGraphNode_StateMachine_2.Locomotion.AnimStateNode_7.GrappleFireLoop
  nodes [Output Animation Pose(Result=Sequence Player 'OctopusVehicle_GrappleFire_Pose'() {sequence=&/OctopusVehicle_GrappleFire_Pose})]
 }
 AnimationTransitionGraph_0 {path=VehicleFinalPoseOverride.AnimGraphNode_StateMachine_2.Locomotion.AnimStateTransitionNode_0.AnimationTransitionGraph_0, nodes [Result()]}
 AnimationTransitionGraph_1 {
  path=VehicleFinalPoseOverride.AnimGraphNode_StateMachine_2.Locomotion.AnimStateTransitionNode_1.AnimationTransitionGraph_1
  nodes [Result(bCanEnterTransition=Get bIsBoosting() -> (bIsBoosting:bool))]
 }
 Transition {path=VehicleFinalPoseOverride.AnimGraphNode_StateMachine_2.Locomotion.AnimStateTransitionNode_10.Transition, nodes [Result()]}
 Transition {
  path=VehicleFinalPoseOverride.AnimGraphNode_StateMachine_2.Locomotion.AnimStateTransitionNode_11.Transition
  nodes [Result(bCanEnterTransition=Get bIsTowhookHolstered() -> (bIsTowhookHolstered:bool))]
 }
 Transition {
  path=VehicleFinalPoseOverride.AnimGraphNode_StateMachine_2.Locomotion.AnimStateTransitionNode_12.Transition
  nodes [Result(bCanEnterTransition=Get bShouldPlayGrappleFire() -> (bShouldPlayGrappleFire:bool))]
 }
 Transition {
  path=VehicleFinalPoseOverride.AnimGraphNode_StateMachine_2.Locomotion.AnimStateTransitionNode_13.Transition
  nodes [Result(bCanEnterTransition=Get bShouldPlayGrappleFire())]
 }
 Transition {
  path=VehicleFinalPoseOverride.AnimGraphNode_StateMachine_2.Locomotion.AnimStateTransitionNode_14.Transition
  nodes [Result(bCanEnterTransition=Get bIsBoosting())]
 }
 Transition {
  path=VehicleFinalPoseOverride.AnimGraphNode_StateMachine_2.Locomotion.AnimStateTransitionNode_15.Transition
  nodes [Result(bCanEnterTransition=Get bIsBoosting())]
 }
 Transition {
  path=VehicleFinalPoseOverride.AnimGraphNode_StateMachine_2.Locomotion.AnimStateTransitionNode_16.Transition
  nodes [Result(bCanEnterTransition=Get bIsTowhookAttached() -> (bIsTowhookAttached:bool))]
 }
 Transition {
  path=VehicleFinalPoseOverride.AnimGraphNode_StateMachine_2.Locomotion.AnimStateTransitionNode_17.Transition
  nodes [Result(bCanEnterTransition=Get bIsTowhookAttached())]
 }
 Transition {
  path=VehicleFinalPoseOverride.AnimGraphNode_StateMachine_2.Locomotion.AnimStateTransitionNode_18.Transition
  nodes [Result(bCanEnterTransition=Get bShouldPlayGrappleFire())]
 }
 Transition {
  path=VehicleFinalPoseOverride.AnimGraphNode_StateMachine_2.Locomotion.AnimStateTransitionNode_19.Transition
  nodes [Result(bCanEnterTransition=Get bShouldPlayGrappleFire())]
 }
 Transition {
  path=VehicleFinalPoseOverride.AnimGraphNode_StateMachine_2.Locomotion.AnimStateTransitionNode_2.Transition
  nodes [Result(bCanEnterTransition=NOT Boolean(A=Get bIsBoosting()))]
 }
 Transition {path=VehicleFinalPoseOverride.AnimGraphNode_StateMachine_2.Locomotion.AnimStateTransitionNode_20.Transition, nodes [Result()]}
 Transition {path=VehicleFinalPoseOverride.AnimGraphNode_StateMachine_2.Locomotion.AnimStateTransitionNode_4.Transition, nodes [Result()]}
 Transition {
  path=VehicleFinalPoseOverride.AnimGraphNode_StateMachine_2.Locomotion.AnimStateTransitionNode_5.Transition
  nodes [Result(bCanEnterTransition=Get bIsBoosting())]
 }
 Transition {
  path=VehicleFinalPoseOverride.AnimGraphNode_StateMachine_2.Locomotion.AnimStateTransitionNode_6.Transition
  nodes [Result(bCanEnterTransition=NOT Boolean(A=Get bIsLowerVelocity()))]
 }
 Transition {
  path=VehicleFinalPoseOverride.AnimGraphNode_StateMachine_2.Locomotion.AnimStateTransitionNode_7.Transition
  nodes [Result(bCanEnterTransition=Get bEnterMovingInPlace() -> (bEnterMovingInPlace:bool))]
 }
 Transition {
  path=VehicleFinalPoseOverride.AnimGraphNode_StateMachine_2.Locomotion.AnimStateTransitionNode_8.Transition
  nodes [Result(bCanEnterTransition=Get bShouldPlayPivotTransition() -> (bShouldPlayPivotTransition:bool))]
 }
 Transition {
  path=VehicleFinalPoseOverride.AnimGraphNode_StateMachine_2.Locomotion.AnimStateTransitionNode_9.Transition
  nodes [Result(bCanEnterTransition=Get bShouldPlayGrappleFire())]
 }
 VehicleFullBodyOverride {nodes [
   Input Pose() {name=InPoseFullBody}
   Output Pose() {layerGroup=VehicleOverride}
  ]}
 VehicleLowerBodyOverride {nodes [
   Input Pose() {name=InPoseLowerBodyDefault}
   Input Pose() {name=InPoseUpperBody}
   Output Pose() {layerGroup=VehicleOverride}
  ]}
 VehicleSplitBodyOverride {nodes [
   Input Pose() {name=InPoseSplitBody}
   Input Pose() {name=InPoseUpperAndLowerBody}
   Input Pose() {name=InPoseUpperAndLowerBodyRemoveRoot}
   Output Pose() {layerGroup=Vehicle Override}
  ]}
]
variables []
